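# Translucent unlit Niagara sprite material for the ToyFox pilot teleport beam, with color selectable per particle.
# Fast-panning noise distorts the horizontal gradient that samples a curve atlas for glow intensity, multiplied by particle color and a ColorIndex-selected atlas color.
# Opacity combines an undistorted curve profile, a scaled soft sphere mask and particle alpha.
# A WipeOut dynamic parameter wipes it along the vertical gradient, and depth fade softens intersections.
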Material /ToyFoxGameplay/Assets/FX/PilotTeleport/PerColor/M_ToyFox_PolotTeleportBeam
graph {
 properties {
  bAutomaticallySetUsageInEditor=false
  bUsedWithNiagaraSprites=true
  blendMode=BLEND_Translucent
  shadingModel=MSM_Unlit
  usedShadingModels=Unlit
 }
 nodes [
  n1 = TextureCoordinate()
  n2 = Constant2Vector() {g=1.2, r=1}
  n3 = TextureSample(UVs=Panner(Coordinate=Multiply(A=@n1, B=@n2)) {speedX=0.2, speedY=5}) {texture=/BRCosmetics/Gadgets/Assets/VinderTech_GliderChute/Glider_SlickSwish/FX/T_Glider_SlickSwish_noise}
  n4 = Add(
   A=MaterialFunctionCall() {materialFunction=&LinearGradient}.UGradient
   B=Divide(A=LinearInterpolate(Alpha=@n3.R) {constA=-1}, B=ScalarParameter() {defaultValue=20})
  )
  Add(A=@n4) {constB=0}
  n5 = ParticleColor()
  n6 = MaterialFunctionCall() {materialFunction=&LinearGradient}
  n7 = MaterialFunctionCall(UVs=TextureCoordinate(), Texture Scale=Constant2Vector() {g=2, r=1.3}) {materialFunction=&ScaleUVsByCenter}
  n8 = Constant2Vector() {g=0.75, r=0.5}
  n9 = Power(Base=Multiply(A=SphereMask(A=@n7.UVs, B=@n8) {attenuationRadius=0.5, hardnessPercent=0}, B=ScalarParameter() {defaultValue=1, parameterName=Param_4}))
  n10 = Multiply(
   A=Saturate(None=CurveAtlasRowParameter(CurveTime=@n6.UGradient) {atlas=/ToyFoxGameplay/Assets/FX/CA_ToyFox_VFX, curve=/ToyFoxGameplay/Assets/FX/Curve_ToyFox_Color01, parameterName=Param_2}.G)
   B=Multiply(A=Saturate(None=@n9), B=ScalarParameter() {defaultValue=3, parameterName=Param_9})
  )
  n11 = OneMinus(None=MaterialFunctionCall() {materialFunction=&LinearGradient}.VGradient)
  n12 = DynamicParameter() {parameterIndex=3, defaultValue=<rgba 0 1 1 1>, paramNames [ColorIndex,Param2,Param3,WipeOut]}
  n13 = Multiply(
   A=Subtract(A=Saturate(None=@n11), B=LinearInterpolate(Alpha=@n12.WipeOut) {constA=1, constAlpha=1, constB=-1})
   B=ScalarParameter() {defaultValue=8, parameterName=Param_3}
  )
  DepthFade(
   Opacity=Multiply(A=Multiply(A=@n5.A, B=@n10), B=Saturate(None=@n13))
   FadeDistance=ScalarParameter() {defaultValue=250, parameterName=DepthFadeDis}
  )
  MaterialFunctionCall() {materialFunction=&DebugTimeSine}
  n14 = CurveAtlasRowParameter(CurveTime=DynamicParameter() {parameterIndex=3, defaultValue=<rgba 0 1 1 1>, paramNames [ColorIndex,Param2,Param3,WipeOut]}.ColorIndex) {atlas=&CA_ToyFox_VFX, curve=/ToyFoxGameplay/Assets/FX/Curve_ToyFox_Color02, defaultValue=1, parameterName=Param_0}
  n15 = ScalarParameter() {defaultValue=20, parameterName=Param_1}
  n16 = CurveAtlasRowParameter(CurveTime=@n4) {atlas=&CA_ToyFox_VFX, curve=&Curve_ToyFox_Color01, parameterName=Param_2}
  MaterialFunctionCall(Emissive=Multiply(A=Multiply(A=@n14, B=@n15), B=Multiply(A=@n5.RGB, B=@n16.B))) {materialFunction=&MF_FX_EmissiveAdjustments}
 ]
}
